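# Niagara swing trail effect for the TerrierCure drumstick pickaxe cosmetic, drawn by a single ribbon emitter.
# Anim-trail modules spawn, initialize and update short-lived particles from bone or socket indices, taking their color from a system-level time-of-day color.
# Over each particle's life the ribbon narrows and fades along curves, with RGB and alpha also scaled by time of day, including a blue tint at night.
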NiagaraSystem /SparksCosmetics/Instrument/DrumStick/TerrierCure/Effects/NS_Pickaxe_DrumStick_TerrierCure_Trail
dependencies {
 usedDataInterfaces [&NiagaraDataInterfaceCurve, &NiagaraDataInterfaceColorCurve]
 usedDynamicInputs [&ColorFromCurve]
 usedModules [
  &self:SystemSpawnScript.SystemScriptSource.SystemScriptGraph.NiagaraNodeAssignment_0.SetVariables_#1
  &SystemState
 ]
}
emitters {Trail_Instrument_Color001 {
  emitterUpdateScript [
   &EmitterState(Life Cycle Mode=::NewEnumerator1, Loop Behavior=::NewEnumerator1, Loop Duration=0.22, UseLoopDelay=true)
   &FN_SpawnAnimTrail_v2()
  ]
  particleSpawnScript [
   /Game/Effects/Niagara/Modules/AnimTrail/FN_InitializeAnimTrail/FN_InitializeAnimTrailBoneIndecies(First Bone or Socket Index=1)
   /Game/Effects/Niagara/Modules/AnimTrail/FN_InitializeAnimTrail/FN_InitializeAnimTrail(AnimTrailWidth=1, Color=$System.TOD Color, Lifetime=0.2)
   &SystemLocation(MaskSpawnGroup=false)
  ]
  particleUpdateScript [
   &ScaleRibbonWidth(Ribbon Width Scale=&FloatFromCurve(FloatCurve=DataInterface() {propertyValues {Curve {keys [
         {arriveTangent=-1.2579, interpMode=RCIM_Cubic, time=0, value=1}
         {arriveTangent=-0.9664, interpMode=RCIM_Cubic, leaveTangent=-0.9664, tangentMode=RCTM_User, time=1, value=0}
        ]}}}))
   &ScaleColor(
    Scale Alpha=&Multiply_Float(
     A=&FloatFromCurve(FloatCurve=DataInterface() {propertyValues {Curve {keys [
          {interpMode=RCIM_Cubic, leaveTangent=-3.048, tangentMode=RCTM_Break, time=0, value=1}
          {arriveTangent=-0.7468, interpMode=RCIM_Cubic, leaveTangent=-0.7468, tangentMode=RCTM_User, time=0.5, value=0.1}
          {arriveTangent=-0.2139, interpMode=RCIM_Cubic, tangentMode=RCTM_Break, time=1, value=0}
         ]}}})
     B=&FNTimeOfDayScalar(Night Scale=0.45, Morning Scale=0.58, Day Scale=0.55, Evening Scale=0.55)
    )
    Scale RGB=&FN_Select_Vector_from_Time_Of_Day(Morning=<0.9 0.9 0.9>, Day=<1 1 1>, Evening=<1 1 1>, Night=<0.5 0.5 1>)
    ScaleA=true
    ScaleRGB=true
   ) {module=ScaleColor002}
   &FN_UpdateAnimTrail_v2()
  ]
  rendererClasses [&NiagaraRibbonRendererProperties]
  renderers [{rendererClass=&NiagaraRibbonRendererProperties}]
 }}
summary {
 systemName=NS_Pickaxe_DrumStick_TerrierCure_Trail
 emitters [{emitterName=Trail_Instrument_Color001, rendererClasses [&NiagaraRibbonRendererProperties]}]
 userVariables [
  User.RibbonIDOffset {type=NiagaraInt32}
  User.SkeletalMesh {
   type=NiagaraDataInterfaceSkeletalMesh
   defaultValue {
    struct=NiagaraExt_VariableValue_DataInterface
    value {dataInterface=&self:NiagaraDataInterfaceSkeletalMesh_0}
   }
  }
  User.TrailColor {type=LinearColor, description="", defaultValue {struct=LinearColor, value=<rgba 1 1 1 0.25>}}
  User.TrailSecondaryColor {type=LinearColor, description="", defaultValue {struct=LinearColor, value=<rgba 10 10 10 0.25>}}
 ]
}
